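# Conversation database for the CerealBox NPC, defining its excitable, Unicorn Flakes-themed greeting.
# The entry point leads to one speech node with a default line plus contextual lines gated by time since last match, first-ever conversation, held bars, and the CerealBox outfit.
# Curve rows set line priority, and two outfit reactions split on this match's conversation count.
# The speech node then links out to the shared NPC quests and basic services conversations.
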ConversationDatabase /NPCLibrary/NPCs/CerealBox/CerealBox/NPCLibrary_Conversation_CerealBox
entry_points [{tag=Conversation.Entry.CerealBox, destinations [#1]}]
nodes [
 ConversationEntryPointNode_7 {
  id=#1
  class=ConversationEntryPointNode
  outputs [#2]
  properties {entryTag {tagName=Conversation.Entry.CerealBox}}
 }
 Speech_C_4 {
  id=#2
  class=Speech_C
  outputs [#3, #4]
  properties {
   nodeName=""
   generalConfig {
    defaultMessage="EVERYBODY GET SOOOOME UNICORN FLAKES!!!"
    contextualMessages [
     {
      message="HIYA FRIEND! MY HOOF SHOOTS CEREAL FOREVER!! "
      contextRequirements [{requirement {class=FortControllerRequirement_CheckTimeSinceLastPlayedMatch}}]
      priority {curve {rowName=Tandem.Speech.Priority.WhileSincePlayed}}
     }
     {
      message="HIYA!! ANY IDEA HOW TO BREAK CEREAL-BASED CURSES???"
      contextRequirements [{requirement {class=FortControllerRequirement_IsFirstEverConversationWithNPC}}]
      priority {curve {rowName=Tandem.Speech.Priority.FirstEverInteraction}}
     }
     {
      message="THOSE BARS COULD BUY 34.45 BOXES OF UNICORN FLAKES!!!"
      contextRequirements [{requirement {class=FortControllerRequirement_HasBars}}]
      priority {curve {rowName=Tandem.Speech.Priority.HighTotalBars}}
     }
     {
      message="OH NO!! THEY GOT YOU TOO?!?"
      contextRequirements [
       {requirement {
         class=FortControllerRequirement_HasCID
         properties {allowedCIds [/BRCosmetics/Athena/Items/Cosmetics/Characters/CID_A_198_Athena_Commando_M_CerealBox]}
        }}
       {requirement {
         class=FortControllerRequirement_CheckNPCMatchConversationCount
         properties {nPCUniqueIdTag {tagName=AISpawnerData.Type.Tandem.CerealBox}, requiredConversationCountConfig {quantity {value=3}, requirement=~}}
        }}
      ]
      priority {curve {rowName=Tandem.Speech.Priority.OutfitReaction}}
     }
     {
      message="HAHA, YOU'RE MOCKING ME THEN?? I THINK THAT'S FUN!!"
      contextRequirements [
       {requirement {
         class=FortControllerRequirement_HasCID
         properties {allowedCIds [&CID_A_198_Athena_Commando_M_CerealBox]}
        }}
       {requirement {
         class=FortControllerRequirement_CheckNPCMatchConversationCount
         properties {nPCUniqueIdTag {tagName=AISpawnerData.Type.Tandem.CerealBox}, requiredConversationCountConfig {quantity {value=3}}}
        }}
      ]
      priority {curve {rowName=Tandem.Speech.Priority.OutfitReaction}}
     }
    ]
   }
  }
 }
 ConversationLinkNode_0 {id=#4, class=ConversationLinkNode, properties {remoteEntryTag {tagName=Conversation.Entry.BasicServices}}}
 ConversationLinkNode_1 {id=#3, class=ConversationLinkNode, properties {remoteEntryTag {tagName=Conversation.Entry.NPCLibrary.Quests}}}
]
